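# Niagara system for the GoldenLychee01 vehicle boost cosmetic, trailing colored sprites and ribbons from the car's boost sockets.
# One sprite emitter feeds three renderers, using spawn groups as visibility tags to layer fast streaks and pulsing stars.
# A second emits slower, pulsing sprites, while a per-socket ribbon emitter adds two trail layers with separate width and color controls.
# All emitters offset by DTCarOffset; painted-variant overrides are present but disabled.
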NiagaraSystem /VehicleCosmetics/Boosts/Boost_GoldenLychee01/FX/NS_Boost_GoldenLychee01
dependencies {
 usedDataInterfaces [
  &NiagaraDataInterfaceArrayColor
  &NiagaraDataInterfaceCurve
  &NiagaraDataInterfaceVector2DCurve
  &NiagaraDataInterfaceSkeletalMesh
 ]
 usedDynamicInputs [&RandomRangeFloat, &Vector2DFromCurve]
 usedModules [
  &self:SystemSpawnScript.SystemScriptSource.SystemScriptGraph.NiagaraNodeAssignment_0.SetVariables_#1
  &SystemState
 ]
}
emitters {
 NE_VC_Boost_Sprite_Parent {
  emitterUpdateScript [
   &EmitterState()
   &VC_FilteredSocketCount()
   &SpawnRate(SpawnRate=50, Use Spawn Probability=false)
   &SpawnRate(SpawnGroup=1, SpawnRate=50, Use Spawn Probability=false) {module=SpawnRate001}
   &SpawnRate(SpawnGroup=2, SpawnRate=20, Use Spawn Probability=false) {module=SpawnRate002}
   &SpawnRate(SpawnGroup=3, SpawnRate=15, Use Spawn Probability=false) {module=SpawnRate003}
  ]
  particleSpawnScript [
   &VC_PaintedVariantsOverride() {enabled=false}
   /SetVariables(Particles.VisibilityTag=$Emitter.SpawnGroup) {isSetParameters=true}
   &InitializeParticle(
    Color=&SelectLinearColorFromArray(
     Array Sampling Mode=::NewEnumerator0
     Direct Array Index=?
     Color Selection Array=DataInterface() {propertyValues {ColorData [<rgba 0.349 1 0.7762 1>, <rgba 0.9124 0.3594 1 1>, <rgba 1 0.9311 0.3958 1>]}}
     Random Seed=0
    )
    Lifetime=&MergeHelper_FloatSwitch(0=0.3, 1=0.3, 2=0.4, 3=?, Index=&SpawnGroup(), Num=::NewEnumerator1)
    Position Mode=::NewEnumerator0
    Sprite Rotation Angle=0.4
    Sprite Rotation Mode=::NewEnumerator1
    Sprite Size=&RandomRangeVector2D(Minimum=<2 200>, Maximum=<2 250>)
    Sprite Size Mode=::NewEnumerator4
    Write Material Random=true
   )
   &PartitionParticles(Number Of Partitions=$Emitter.PortCount, Particles Per Partition=2, Partition Mode:ENiagara_PartitionMode=::NewEnumerator1)
   &SkeletalMeshLocation(
    Bone / Socket Index=$Particles.PartitionParticles.Partition
    Bone / Socket Sampling Mode=::NewEnumerator5
    MaskSpawnGroup=false
    Mesh Sampling Type=::NewEnumerator6
    Sampled Position Offset=&AddVectorToPosition(Position=$User.DTCarOffset, Vector=<-25 0 0>)
    Skeletal Mesh=$System.SkeletalMesh
   )
   &AddVelocity(Velocity=&MergeHelper_VectorSwitch(0=<-1250 0 0>, 1=<-1250 0 0>, 2=<-1250 0 0>, 3=?, Index=&SpawnGroup(), Num=::NewEnumerator1))
   &ShapeLocation(
    MaskSpawnGroup=false
    Offset=<-80 0 0>
    Offset Coordinate Space=::Local
    Offset Mode=::NewEnumerator0
    Override Seed=false
    Shape Origin=$Particles.Position
    Sphere Radius=30
    UseOwnerScale=true
   )
  ]
  particleUpdateScript [
   /SetVariables(
    Particles.DotEdgeColor:LinearColor=<rgba 1 0.0279 0.9278 1>
    Particles.DotOpacityColor:LinearColor=&MultiplyLinearColor(
     Linear Color=&RandomRangeLinearColor(Minimum=<rgba 0.2831 0.2831 1 1>, Maximum=<rgba 0.0666 0.6867 1 1>)
     Scale Factor=&MakeVector4FromVectorAndFloat(
      "Vector (RGB)"=&VectorFromFloat(Value=1.5)
      "Float (Alpha)"=&FloatFromCurve(FloatCurve=DataInterface() {propertyValues {Curve {keys [<at 0.5 0.2 RCIM_Cubic>,<at 1 0 RCIM_Cubic>]}}})
     )
    )
    Particles.DotsOpacitySize:Vector2f=&RandomRangeVector2D(Minimum=<15 388>, Maximum=<20 400>)
    Particles.StarSize:Vector2f=&Vector2DFromFloat(Value=&FloatFromCurve(
      FloatCurve=DataInterface() {propertyValues {Curve {
         defaultValue=50
         keys [
          {interpMode=RCIM_Cubic, tangentMode=RCTM_User, time=0, value=1}
          {interpMode=RCIM_Cubic, tangentMode=RCTM_User, time=0.1, value=2}
          {arriveTangent=-3.3333, interpMode=RCIM_Cubic, leaveTangent=-3.3333, time=0.25, value=1.75}
          {arriveTangent=16.1909, interpMode=RCIM_Cubic, leaveTangent=16.1909, tangentMode=RCTM_User, time=0.4, value=1}
          {arriveTangent=16.3709, interpMode=RCIM_Cubic, leaveTangent=16.3709, tangentMode=RCTM_User, time=0.6, value=1.5}
          {arriveTangent=-3.75, interpMode=RCIM_Cubic, leaveTangent=-3.75, time=0.75, value=0.5}
          {tangentMode=RCTM_User, time=1, value=0}
         ]
        }}}
      Scale Curve=50
     ))
   ) {isSetParameters=true}
   &ParticleState()
   &SolveForcesAndVelocity(Clamp Velocity=false, Limit Acceleration=false)
  ]
  rendererClasses [&NiagaraSpriteRendererProperties]
  renderers [
   {rendererClass=&NiagaraSpriteRendererProperties}
   {rendererClass=&NiagaraSpriteRendererProperties, rendererIndex=1}
   {rendererClass=&NiagaraSpriteRendererProperties, rendererIndex=2}
  ]
 }
 NE_VC_Boost_Sprite_Parent001 {
  emitterUpdateScript [
   &EmitterState()
   &VC_FilteredSocketCount()
   &SpawnRate(SpawnRate=20, Use Spawn Probability=false)
  ]
  particleSpawnScript [
   &VC_PaintedVariantsOverride() {enabled=false}
   &InitializeParticle(
    Color=&SelectLinearColorFromArray(
     Array Sampling Mode=::NewEnumerator0
     Direct Array Index=?
     Color Selection Array=DataInterface() {propertyValues {ColorData [<rgba 0.349 1 0.7762 1>, <rgba 0.9124 0.3594 1 1>, <rgba 1 0.9311 0.3958 1>]}}
     Random Seed=0
    )
    Sprite Rotation Mode=::NewEnumerator2
    Sprite Size Mode=::NewEnumerator1
    Uniform Sprite Size Max=60
    Uniform Sprite Size Min=50
    UsePositionOffset=false
    Write Material Random=true
   )
   &PartitionParticles(Number Of Partitions=$Emitter.PortCount)
   &SkeletalMeshLocation(
    Bone / Socket Index=$Particles.PartitionParticles.Partition
    Bone / Socket Sampling Mode=::NewEnumerator5
    MaskSpawnGroup=false
    Mesh Sampling Type=::NewEnumerator6
    Sampled Position Offset=&AddVectorToPosition(Position=$User.DTCarOffset, Vector=<-30 0 0>)
    Skeletal Mesh=$System.SkeletalMesh
   )
   &AddVelocity(Velocity=<-150 0 0>)
   &ShapeLocation(
    MaskSpawnGroup=false
    Offset=<-20 0 0>
    Offset Mode=::NewEnumerator0
    Override Seed=false
    Shape Origin=$Particles.Position
    Sphere Radius=50
    UseOwnerScale=true
   )
  ]
  particleUpdateScript [
   &ParticleState()
   &ScaleSpriteSize(
    Scale Sprite Size Mode=::NewEnumerator0
    Uniform Scale Factor=&FloatFromCurve(FloatCurve=DataInterface() {propertyValues {Curve {
        defaultValue=50
        keys [
         {interpMode=RCIM_Cubic, tangentMode=RCTM_User, time=0, value=1}
         {interpMode=RCIM_Cubic, tangentMode=RCTM_User, time=0.1, value=2}
         {arriveTangent=-3.3333, interpMode=RCIM_Cubic, leaveTangent=-3.3333, time=0.25, value=1.5}
         {arriveTangent=16.0587, interpMode=RCIM_Cubic, leaveTangent=16.0587, tangentMode=RCTM_User, time=0.4, value=1}
         {arriveTangent=16.3709, interpMode=RCIM_Cubic, leaveTangent=16.3709, tangentMode=RCTM_User, time=0.6, value=1.5}
         {arriveTangent=-3.75, interpMode=RCIM_Cubic, leaveTangent=-3.75, time=0.75, value=1}
         {tangentMode=RCTM_User, time=1, value=0}
        ]
       }}})
   )
   &SolveForcesAndVelocity(Clamp Velocity=false, Limit Acceleration=false)
  ]
  rendererClasses [&NiagaraSpriteRendererProperties]
  renderers [{rendererClass=&NiagaraSpriteRendererProperties}]
 }
 Ribbon {
  emitterUpdateScript [
   &EmitterState()
   &VC_FilteredSocketCount()
   &VC_RateQualityHandler(Epic Spawn Rate=25, High Spawn Rate=25, Low Spawn Rate=25, Medium Spawn Rate=25, Use Custom Rate?=true)
   &VC_SpawnRibbon_v1(Number of Ribbons=$Emitter.PortCount, Spawn Time=-1, SpawnRate=$Emitter.Spawn Rate)
  ]
  particleSpawnScript [
   &VC_PaintedVariantsOverride() {enabled=false}
   &VC_InitializeRibbon_v1(
    Custom Offset=true
    Enable Skeletal Position=true
    Lifetime=1
    Local Vector Offset=&AddVectorToPosition(Position=<10 0 0>, Vector=Linked() {linkedVariable {name=User.DTCarOffset, type=NiagaraPosition}})
    Offset Amount=1
    Ribbon Width=48
   )
   &AddVelocity(Velocity=<-1000 0 0>) {module=AddVelocity001}
   /SetVariables(Particles.Width_NoiseTrail=15) {isSetParameters=true}
  ]
  particleUpdateScript [
   &ParticleState()
   &InheritSourceMovement(
    Applied Position Scale=&Multiply_VectorByFloat(
     Vector=<1 1 1>
     Float=&FloatFromCurve(FloatCurve=DataInterface() {propertyValues {Curve {keys [{interpMode=RCIM_Cubic, tangentMode=RCTM_Break, time=0.5, value=1}, {interpMode=RCIM_Cubic, tangentMode=RCTM_Break, time=1, value=0}]}}})
    )
    ApplyForce=false
    ApplyPosition=true
    ApplyPositionFalloff=false
   )
   &SolveForcesAndVelocity(Clamp Velocity=false, Limit Acceleration=false)
   &VC_UpdateRibbon_v1(Enable Skeletal Position=true)
   &ScaleColor(Scale RGB=&VectorFromFloat(Value=1), ScaleA=true, ScaleRGB=true)
   &DynamicMaterialParameters(
    Index 0 Param 1=&FloatFromCurve(FloatCurve=DataInterface() {propertyValues {Curve {keys [<at 0 0.75>,<at 0.2 0.96>,<at 0.7 0.92>,<at 1 0.7>]}}})
    Module.Index0Float0Write=true
    Module.Index0Float1Write=true
    Module.Index0Float2Write=true
    Module.Index0Float3Write=true
   )
   &ScaleRibbonWidth(Ribbon Width Scale=&FloatFromCurve(FloatCurve=DataInterface() {propertyValues {Curve {keys [<at 0.1 1>,<at 1 0.8>]}}}))
   /SetVariables(
    Particles. DynamicIndex0_NoiseTrail:Vector4f=&MakeVector4(
     X=&FloatFromCurve(FloatCurve=DataInterface() {propertyValues {Curve {keys [<at 0 0.85>,<at 0.2 0.96>,<at 0.7 0.93>,<at 1 0.3>]}}})
     Y=0
     Z=0
     W=0
    )
    Particles.ScaleNoiseTrail:NiagaraFloat=&Multiply_Float(
     A=$Particles.Initial.Width_NoiseTrail
     B=&FloatFromCurve(FloatCurve=DataInterface() {propertyValues {Curve {keys [<at 0.4 1>,<at 1 0.7>]}}})
    )
    Particles.Scale_NoiseTrail:LinearColor=&MultiplyLinearColor(
     Linear Color=$Particles.Color_NoiseTrail
     Scale Factor=&MakeVector4(
      X=40
      Y=10
      Z=10
      W=&FloatFromCurve(FloatCurve=DataInterface() {propertyValues {Curve {keys [<at 0 -0.1>,<at 0.5 1>,<at 0.7 1>,<at 0.8 0.31>]}}})
     )
    )
   ) {isSetParameters=true}
   /SetVariables(Particles.GradiantFillColor:LinearColor=<rgba 0 0.8333 1 0.2>) {isSetParameters=true}
  ]
  rendererClasses [&NiagaraRibbonRendererProperties]
  renderers [
   {rendererClass=&NiagaraRibbonRendererProperties}
   {rendererClass=&NiagaraRibbonRendererProperties, rendererIndex=1}
  ]
 }
}
summary {
 systemName=NS_Boost_GoldenLychee01
 emitters [
  {emitterName=NE_VC_Boost_Sprite_Parent, rendererClasses [&NiagaraSpriteRendererProperties]}
  {emitterName=NE_VC_Boost_Sprite_Parent001, rendererClasses [&NiagaraSpriteRendererProperties]}
  {emitterName=Ribbon, rendererClasses [&NiagaraRibbonRendererProperties]}
 ]
 userVariables [
  User.bIsFrontendPreview {type=NiagaraBool}
  User.bIsPainted {type=NiagaraBool}
  User.ColorVariantID {type=NiagaraInt32}
  User.DTCarOffset {type=NiagaraPosition, description="", defaultValue {}}
  User.intensity_trail {type=NiagaraFloat, defaultValue=3, description=""}
  User.Lines_ribbon {type=NiagaraFloat, defaultValue=5, description=""}
  User.PaintedColor {type=LinearColor, description="", defaultValue {struct=LinearColor, value=<rgba 1 1 1 1>}}
  User.PaintedColor2 {type=LinearColor, description="", defaultValue {struct=LinearColor, value=<rgba 1 1 1 1>}}
  User.PaintedScalar {type=NiagaraFloat, defaultValue=1, description=""}
  User.PortCount {type=NiagaraInt32, defaultValue=1}
 ]
}
